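# Sound cue for the Six Shooter pistol's out-of-ammo sound.
# A quality switch picks randomly among four equally weighted waves on its first two levels and plays only the 01 wave on the last two.
# A modulator above the switch varies volume slightly while keeping pitch fixed.
# It uses weapon reload attenuation, a third-person ranged-fire sound class, and a concurrency override capping it at two instances, stopping the oldest.
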SoundCue /WeaponsUnvaulted/Assets/Audio/SixShooter_Pistol/Pistol_OutofAmmo_Cue
root=Modulator_0
nodes [
 {id=Modulator_0, children [QualityLevel_0], properties {pitchMax=1, pitchMin=1, volumeMax=0.9, volumeMin=0.8}}
 {id=QualityLevel_0, children [Random_1,Random_1,WavePlayer_7,WavePlayer_7]}
 {id=Random_1, children [WavePlayer_4,WavePlayer_5,WavePlayer_6,WavePlayer_7], properties {weights [1,1,1,1]}}
 {id=WavePlayer_4, properties {soundWaveAssetPtr=/WeaponsUnvaulted/Assets/Audio/SixShooter_Pistol/SoundWaves/Pistol_OutofAmmo_04}}
 {id=WavePlayer_5, properties {soundWaveAssetPtr=/WeaponsUnvaulted/Assets/Audio/SixShooter_Pistol/SoundWaves/Pistol_OutofAmmo_03}}
 {id=WavePlayer_6, properties {soundWaveAssetPtr=/WeaponsUnvaulted/Assets/Audio/SixShooter_Pistol/SoundWaves/Pistol_OutofAmmo_02}}
 {id=WavePlayer_7, properties {soundWaveAssetPtr=/WeaponsUnvaulted/Assets/Audio/SixShooter_Pistol/SoundWaves/Pistol_OutofAmmo_01}}
]
properties {
 attenuationSettings=&Fort_Weapon_Reload_Att
 bOverrideConcurrency=true
 maxDistance=2250
 soundClassObject=&Fort_WeaponRanged_Fire_3P_PSM
 volumeMultiplier=0.6
 attenuationOverrides {falloffDistance=1600}
 concurrencyOverrides {maxCount=2, resolutionRule=StopOldest}
}
