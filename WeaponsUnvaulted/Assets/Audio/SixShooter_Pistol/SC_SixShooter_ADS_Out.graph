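# Sound cue for the six-shooter pistol's aim-down-sights exit foley.
# A single wave player feeds a modulator with a fixed volume reduction and no random variation.
# The cue lowers volume further with its own multiplier, uses the weapon reload attenuation, and routes to the weapon foley sound class.
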SoundCue /WeaponsUnvaulted/Assets/Audio/SixShooter_Pistol/SC_SixShooter_ADS_Out
root=Modulator_0
nodes [
 {id=Modulator_0, children [WavePlayer_1], properties {volumeMax=0.7, volumeMin=0.7}}
 {id=WavePlayer_1, properties {soundWaveAssetPtr=/WeaponsUnvaulted/Assets/Audio/SixShooter_Pistol/SoundWaves/SixShooter_ADS_Out}}
]
properties {
 attenuationSettings=&Fort_Weapon_Reload_Att
 maxDistance=2250
 soundClassObject=&Fort_WeaponFoley
 volumeMultiplier=0.2
}
