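# Sound cue for the equip sound of the Six Shooter pistol, classed as weapon foley.
# A single wave player feeds a Modulator left at its default settings for pitch and volume variation, with the cue's volume reduced.
# Spatial falloff comes from the shared weapon reload attenuation asset.
# The cue overrides concurrency to block new instances once its limit is reached.
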SoundCue /WeaponsUnvaulted/Assets/Audio/SixShooter_Pistol/SC_SixShooter_Equip
root=Modulator_0
nodes [
 {id=Modulator_0, children [WavePlayer_0]}
 {id=WavePlayer_0, properties {soundWaveAssetPtr=/WeaponsUnvaulted/Assets/Audio/SixShooter_Pistol/SoundWaves/SixShooter_Equip}}
]
properties {
 attenuationSettings=&Fort_Weapon_Reload_Att
 bOverrideConcurrency=true
 maxDistance=2250
 soundClassObject=&Fort_WeaponFoley
 volumeMultiplier=0.501
 attenuationOverrides {falloffDistance=2000, nonSpatializedRadiusStart=200, attenuationShapeExtents {x=500}}
 concurrencyOverrides {resolutionRule=PreventNew}
}
