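# Sound cue for the Six Shooter pistol's reload insert sound.
# A single wave player feeds the output directly with one insert recording, with no randomization or modulation nodes.
# It uses the shared weapon reload attenuation and weapon foley sound class, with slightly reduced volume.
# A concurrency override prevents new instances from starting once the limit is reached, rather than cutting off playing ones.
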SoundCue /WeaponsUnvaulted/Assets/Audio/SixShooter_Pistol/SC_SixShooter_Reload_Insert
root=WavePlayer_0
nodes [{id=WavePlayer_0, properties {soundWaveAssetPtr=/WeaponsUnvaulted/Assets/Audio/SixShooter_Pistol/SoundWaves/SixShooter_ReloadInsert_01}}]
properties {
 attenuationSettings=&Fort_Weapon_Reload_Att
 bOverrideConcurrency=true
 maxDistance=2250
 soundClassObject=&Fort_WeaponFoley
 volumeMultiplier=0.85
 concurrencyOverrides {resolutionRule=PreventNew}
}
