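# SoundCue for the SixShooter pistol's reload shell impact sound.
# A Random node picks with equal weight between two ReloadStartB wave variations.
# It plays very quietly in the weapon foley sound class with weapon reload attenuation.
# Its concurrency override blocks new instances at the limit rather than stopping playing ones.
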SoundCue /WeaponsUnvaulted/Assets/Audio/SixShooter_Pistol/SC_SixShooter_Reload_ShellImpact
root=Random_0
nodes [
 {id=Random_0, children [WavePlayer_1,WavePlayer_2], properties {weights [1,1]}}
 {id=WavePlayer_1, properties {soundWaveAssetPtr=/WeaponsUnvaulted/Assets/Audio/SixShooter_Pistol/SoundWaves/SixShooter_ReloadStartB_01}}
 {id=WavePlayer_2, properties {soundWaveAssetPtr=/WeaponsUnvaulted/Assets/Audio/SixShooter_Pistol/SoundWaves/SixShooter_ReloadStartB_02}}
]
properties {
 attenuationSettings=&Fort_Weapon_Reload_Att
 bOverrideConcurrency=true
 maxDistance=2250
 soundClassObject=&Fort_WeaponFoley
 volumeMultiplier=0.05
 concurrencyOverrides {resolutionRule=PreventNew}
}
