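# Sound cue for the start of the SixShooter pistol reload in the WeaponsUnvaulted content.
# A single wave player feeds a modulator that slightly randomizes pitch and volume on each play.
# It routes through the weapon foley sound class with shared weapon reload attenuation.
# A concurrency override blocks new instances once the limit is reached rather than stopping ones already playing.
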SoundCue /WeaponsUnvaulted/Assets/Audio/SixShooter_Pistol/SC_SixShooter_Reload_Start
root=Modulator_0
nodes [
 {id=Modulator_0, children [WavePlayer_0], properties {pitchMax=1.1, pitchMin=1, volumeMax=1.2, volumeMin=1.1}}
 {id=WavePlayer_0, properties {soundWaveAssetPtr=/WeaponsUnvaulted/Assets/Audio/SixShooter_Pistol/SoundWaves/SixShooter_ReloadStart_02}}
]
properties {
 attenuationSettings=&Fort_Weapon_Reload_Att
 bOverrideConcurrency=true
 maxDistance=2250
 soundClassObject=&Fort_WeaponFoley
 volumeMultiplier=0.8
 concurrencyOverrides {resolutionRule=PreventNew}
}
